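# Niagara system for the Zti vehicle boost cosmetic: blue and red lightning trails with sparks streaming from the car's boost sockets.
# Two ribbon emitters and three sprite emitters follow filtered skeletal-mesh sockets, with mode-masked velocities.
# Merged layers share emitters through spawn groups, extra color bindings and multiple renderers with separate material parameters.
# User parameters substitute a painted color, adjust painted lightning brightness and offset the effect per car.
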NiagaraSystem /VehicleCosmetics/Boosts/Boost_Zti/FX/NS_Boost_Zti
dependencies {
 usedDataInterfaces [
  &NiagaraDataInterfaceCurve
  &NiagaraDataInterfaceSpriteRendererInfo
  &NiagaraDataInterfaceSkeletalMesh
 ]
 usedDynamicInputs []
 usedModules [
  &self:SystemSpawnScript.SystemScriptSource.SystemScriptGraph.NiagaraNodeAssignment_0.SetVariables_#1
  &self:SystemSpawnScript.SystemScriptSource.SystemScriptGraph.NiagaraNodeAssignment_1.SetVariables_#2
  &SystemState
 ]
}
emitters {
 BlueLightning001 {
  emitterUpdateScript [
   &EmitterState()
   &VC_FilteredSocketCount()
   &VC_RateQualityHandler()
   &VC_SpawnRibbon_v1(Number of Ribbons=$Emitter.PortCount, SpawnRate=120)
  ]
  particleSpawnScript [
   &VC_PaintedVariantsOverride() {enabled=false}
   &VC_InitializeRibbon_v1(
    Color=&MakeCustomLinearColorFromBool(Bool=$User.bIsPainted, True Color=$User.PaintedColor, False Color=<rgba 0.2 0.5 1 1>)
    Custom Offset=true
    Enable Skeletal Position=true
    Lifetime=/VehicleCosmetics/FX/Niagara/Modules/VC_Mask_Floats(Coro Value:NiagaraFloat=0.75, FrontEnd Value:NiagaraFloat=0.75, FN Value:NiagaraFloat=0.75)
    Local Vector Offset=&AddVectorToPosition(Position=<0 0 0>, Vector=$User.DTCarOffset)
    Offset Amount=1
    Ribbon Width=60
   )
   &AddVelocity(Velocity=&VC_Mask_Vectors(Coro Value:Vector3f=<-800 0 0>, FrontEnd Value:Vector3f=<-4000 0 0>, FN Value:Vector3f=<-1000 0 0>)) {module=AddVelocity001}
  ]
  particleUpdateScript [
   &ParticleState()
   &ScaleVelocity(
    Coordinate Space:ENiagaraCoordinateSpace=::Simulation
    Velocity Scale=&VectorFromFloat(Value=&FloatFromCurve(FloatCurve=DataInterface() {propertyValues {Curve {keys [{interpMode=RCIM_Cubic, tangentMode=RCTM_User, time=0, value=1}, {interpMode=RCIM_Cubic, tangentMode=RCTM_User, time=1, value=0}]}}}))
   )
   &InheritSourceMovement(
    Applied Position Scale=&Multiply_VectorByFloat(
     Vector=<1 1 1>
     Float=&FloatFromCurve(FloatCurve=DataInterface() {propertyValues {Curve {keys [{interpMode=RCIM_Cubic, tangentMode=RCTM_User, time=0, value=1}, {interpMode=RCIM_Cubic, tangentMode=RCTM_User, time=1, value=0}]}}})
    )
    ApplyForce=false
    ApplyPosition=true
    ApplyPositionFalloff=false
   )
   &ScaleColor(
    Scale Alpha=&Multiply_Float(
     A=&FloatFromCurve(FloatCurve=DataInterface() {propertyValues {Curve {keys [<at 0 1.0113 RCIM_Cubic>, {arriveTangent=-1.0113, interpMode=RCIM_Cubic, leaveTangent=-1.0113, time=0.5, value=0.5}, <at 1 0 RCIM_Cubic>]}}})
     B=0.1
    )
    Scale RGB=&MakeCustomVectorFromBool(Bool=$User.bIsPainted, True Vector=<10 10 10>, False Vector=<10 10 10>)
    ScaleA=true
    ScaleRGB=true
   ) {module=ScaleColor001}
   &DynamicMaterialParameters(
    Index 0 Param 1=&RandomRangeFloat(Minimum=1, Maximum=2)
    Index 0 Param 2=&RandomRangeFloat(Minimum=0, Maximum=1)
    Module.Param0WriteEnabled=true
    Module.Param1WriteEnabled=true
    Module.Param2WriteEnabled=true
    Module.Param3WriteEnabled=true
   )
   &SolveForcesAndVelocity(Clamp Velocity=false, Limit Acceleration=false)
   &VC_UpdateRibbon_v1(Enable Skeletal Position=true)
  ]
  rendererClasses [&NiagaraRibbonRendererProperties]
  renderers [{rendererClass=&NiagaraRibbonRendererProperties}]
 }
 LightningBody_Sprites {
  emitterUpdateScript [
   &EmitterState()
   &VC_FilteredSocketCount()
   &SpawnRate(SpawnRate=20, Use Spawn Probability=false)
   &SpawnRate(SpawnGroup=1, SpawnRate=20, Use Spawn Probability=false) {module=SpawnRate001}
  ]
  particleSpawnScript [
   &VC_PaintedVariantsOverride() {enabled=false}
   &InitializeParticle(
    Color=&MakeCustomLinearColorFromBool(Bool=$User.bIsPainted, True Color=$User.PaintedColor, False Color=<rgba 0.1 0.5 1 1>)
    Lifetime Max=0.3
    Lifetime Min=0.2
    Lifetime Mode=::NewEnumerator1
    Position Offset=<2 0 0>
    Sprite Size Mode=::NewEnumerator1
    Uniform Sprite Size Max=150
    Uniform Sprite Size Min=40
    UsePositionOffset=false
    Write Material Random=true
   )
   &PartitionParticles(Number Of Partitions=$Emitter.PortCount, Partition Mode:ENiagara_PartitionMode=::NewEnumerator5)
   &SkeletalMeshLocation(
    Bone / Socket Index=$Particles.PartitionParticles.Partition
    Bone / Socket Sampling Mode=::NewEnumerator5
    MaskSpawnGroup=false
    Mesh Sampling Type=::NewEnumerator6
    Sampled Position Offset=&AddVectorToPosition(Position=<-100 0 0>, Vector=$User.DTCarOffset)
    Skeletal Mesh=$System.SkeletalMesh
   )
   &ShapeLocation(
    Handle Radius=5
    Large Radius=20
    MaskSpawnGroup=false
    Non Uniform Scale=<1 1 10>
    Override Seed=false
    Rotation Coordinate Space=::Simulation
    Rotation Mode=::NewEnumerator1
    Shape Primitive=::NewEnumerator3
    UseOwnerScale=true
   )
   &AddVelocity(
    Velocity=&MaskVectorBySpawnGroup(
     Masked Vector=&RandomRangeVector(Minimum=<1300 30 30>, Maximum=<500 -30 -30>)
     Passthrough Vector=&RandomRangeVector(Minimum=<1150 40 40>, Maximum=<400 -40 -40>)
    )
    Velocity Speed Scale=-1
   )
   /VehicleCosmetics/FX/Niagara/Modules/MergeEmitters/VC_ExtraBindingColors(
    ExtraColors=1
    R1 Color:LinearColor=&MakeCustomLinearColorFromBool(Bool=$User.bIsPainted, True Color=$User.PaintedColor, False Color=<rgba 1 0.1 0.05 1>)
   )
   /SetVariables(Particles.RedLightningVisTag:NiagaraInt32=&SpawnGroup()) {isSetParameters=true}
  ]
  particleUpdateScript [
   &ParticleState()
   &ScaleColor(
    Scale Alpha=&Multiply_Float(
     A=&FloatFromCurve(FloatCurve=DataInterface() {propertyValues {Curve {keys [<at 0 1.0113 RCIM_Cubic>, {arriveTangent=-1.0113, interpMode=RCIM_Cubic, leaveTangent=-1.0113, time=0.5, value=0.5}, <at 1 0 RCIM_Cubic>]}}})
     B=0.1
    )
    Scale RGB=&MakeCustomVectorFromBool(Bool=$User.bIsPainted, True Vector=<10 10 10>, False Vector=<10 10 10>)
    ScaleA=true
    ScaleRGB=true
   ) {module=ScaleColor001}
   &SubUVAnimation(
    End Frame Range Override=16
    "Sprite Renderer "=DataInterface() {propertyValues {SpriteRenderer=&self:LightningBody_Sprites_2.NiagaraSpriteRendererProperties_0}}
    UseEndFrame=true
    UseStartFrame=true
   )
   &DynamicMaterialParameters(
    Index 0 Param 1=3
    Index 1 Param 1=5
    Module.Param0WriteEnabled=true
    Module.Param0WriteEnabled1=true
    Module.Param1WriteEnabled=true
    Module.Param1WriteEnabled1=true
    Module.Param2WriteEnabled=true
    Module.Param2WriteEnabled1=true
    Module.Param3WriteEnabled=true
    Module.Param3WriteEnabled1=true
    Write Parameter Index 1=true
   )
   &SolveForcesAndVelocity(Clamp Velocity=false, Limit Acceleration=false)
   /SetVariables(
    Particles.RedLightningPosition:NiagaraPosition=&Multiply_Vector(
     A=$Output.SolveForcesAndVelocity.Position
     B=&RandomRangeVector(Minimum=<0.6 0.8 1>, Maximum=<1.6 1.3 1>)
    )
    Particles.RedLightningVelocity:Vector3f=&RandomRangeVector(Minimum=<1150 40 40>, Maximum=<400 -40 -40>)
    Particles.r1 RedLightningColor:LinearColor=&self:VC_Dynamic_ScaleColor(
     Color to Scale:LinearColor=$Particles.Initial.R1 Color
     RGB Scale:Vector3f=&MakeCustomVectorFromBool(Bool=$User.bIsPainted, True Vector=<10 10 10>, False Vector=<10 10 10>)
     Alpha Scale:NiagaraFloat=&Multiply_Float(
      A=&FloatFromCurve(FloatCurve=DataInterface() {propertyValues {Curve {keys [<at 0 1.0113 RCIM_Cubic>, {arriveTangent=-1.0113, interpMode=RCIM_Cubic, leaveTangent=-1.0113, time=0.5, value=0.5}, <at 1 0 RCIM_Cubic>]}}})
      B=0.1
     )
    )
   ) {isSetParameters=true}
  ]
  rendererClasses [&NiagaraSpriteRendererProperties]
  renderers [
   {rendererClass=&NiagaraSpriteRendererProperties}
   {rendererClass=&NiagaraSpriteRendererProperties, rendererIndex=1}
  ]
 }
 Lightning_Sprites {
  emitterUpdateScript [
   &EmitterState()
   &VC_FilteredSocketCount()
   &SpawnRate(SpawnRate=30, Use Spawn Probability=false)
   &SpawnRate(SpawnGroup=1, SpawnRate=30, Use Spawn Probability=false) {module=SpawnRate002}
  ]
  particleSpawnScript [
   &VC_PaintedVariantsOverride() {enabled=false}
   &InitializeParticle(
    Color=&MakeCustomLinearColorFromBool(
     Bool=$User.bIsPainted
     True Color=$User.PaintedColor
     False Color=&MaskLinearColorBySpawnGroup(Masked Linear Color=<rgba 2 0.2 0.1 1>, Passthrough Linear Color=<rgba 0.1 0.5 1 1>)
    )
    Lifetime Max=0.4
    Lifetime Min=0.2
    Lifetime Mode=::NewEnumerator1
    Position Offset=<2 0 0>
    Sprite Rotation Mode=::NewEnumerator2
    Sprite Size Mode=::NewEnumerator1
    Uniform Sprite Size Max=270
    Uniform Sprite Size Min=50
    UsePositionOffset=true
    Write Material Random=true
   )
   &PartitionParticles(Number Of Partitions=$Emitter.PortCount, Partition Mode:ENiagara_PartitionMode=::NewEnumerator5)
   &SkeletalMeshLocation(
    Bone / Socket Index=$Particles.PartitionParticles.Partition
    Bone / Socket Sampling Mode=::NewEnumerator5
    MaskSpawnGroup=false
    Mesh Sampling Type=::NewEnumerator6
    Sampled Position Offset=&AddVectorToPosition(Position=<-50 0 0>, Vector=$User.DTCarOffset)
    Skeletal Mesh=$System.SkeletalMesh
   )
   &ShapeLocation(MaskSpawnGroup=false, Override Seed=false, Sphere Radius=10, UseOwnerScale=true)
   &AddVelocity(
    Velocity=&MaskVectorBySpawnGroup(
     Masked Vector=&RandomRangeVector(Minimum=<150 50 50>, Maximum=<150 -50 -50>)
     Passthrough Vector=&RandomRangeVector(Minimum=<100 30 30>, Maximum=<100 -30 -30>)
    )
    Velocity Speed Scale=-1.5
   )
   &VC_ExtraBindingColors(
    ExtraColors=1
    R1 Color:LinearColor=&MakeCustomLinearColorFromBool(Bool=$User.bIsPainted, True Color=$User.PaintedColor, False Color=<rgba 0.1 0.5 1 1>)
   )
   /SetVariables(Particles.BlueLightningVisTag:NiagaraInt32=&SpawnGroup()) {isSetParameters=true}
  ]
  particleUpdateScript [
   &ParticleState()
   &ScaleColor(
    Scale Alpha=&Multiply_Float(
     A=&FloatFromCurve(FloatCurve=DataInterface() {propertyValues {Curve {keys [<at 0 1.0113 RCIM_Cubic>, {arriveTangent=-1.0113, interpMode=RCIM_Cubic, leaveTangent=-1.0113, time=0.5, value=0.5}, <at 1 0 RCIM_Cubic>]}}})
     B=0.1
    )
    Scale RGB=&MakeCustomVectorFromBool(Bool=$User.bIsPainted, True Vector=<8 8 8>, False Vector=<1 1 1>)
    ScaleA=true
    ScaleRGB=true
   ) {module=ScaleColor001}
   &SubUVAnimation(
    End Frame Range Override=16
    "Sprite Renderer "=DataInterface() {propertyValues {SpriteRenderer=&self:Lightning_Sprites_3.NiagaraSpriteRendererProperties_0}}
    UseEndFrame=true
    UseStartFrame=true
   )
   &DynamicMaterialParameters(
    Index 0 Param 1=7
    Index 1 Param 1=10
    Module.Param0WriteEnabled=true
    Module.Param0WriteEnabled1=true
    Module.Param1WriteEnabled=true
    Module.Param1WriteEnabled1=true
    Module.Param2WriteEnabled=true
    Module.Param2WriteEnabled1=true
    Module.Param3WriteEnabled=true
    Module.Param3WriteEnabled1=true
    Write Parameter Index 1=true
   )
   &SolveForcesAndVelocity(Clamp Velocity=false, Limit Acceleration=false)
  ]
  rendererClasses [&NiagaraSpriteRendererProperties]
  renderers [
   {rendererClass=&NiagaraSpriteRendererProperties}
   {rendererClass=&NiagaraSpriteRendererProperties, rendererIndex=1}
  ]
 }
 Ribbons {
  emitterSpawnScript [/SetVariables(Emitter.LightningRedRW=200, Emitter.LightningSmallRW=60, Emitter.LightningsRW=180) {isSetParameters=true}]
  emitterUpdateScript [
   &EmitterState()
   &VC_FilteredSocketCount()
   &VC_RateQualityHandler()
   &VC_SpawnRibbon_v1(Number of Ribbons=$Emitter.PortCount, SpawnRate=60)
  ]
  particleSpawnScript [
   &VC_PaintedVariantsOverride() {enabled=false}
   &VC_InitializeRibbon_v1(
    Color=&MakeCustomLinearColorFromBool(Bool=$User.bIsPainted, True Color=$User.PaintedColor, False Color=<rgba 0.2 0.102 1 1>)
    Custom Offset=true
    Enable Skeletal Position=true
    Lifetime=&VC_Mask_Floats(Coro Value:NiagaraFloat=0.75, FrontEnd Value:NiagaraFloat=0.75, FN Value:NiagaraFloat=0.75)
    Local Vector Offset=&AddVectorToPosition(Position=<80 0 0>, Vector=$User.DTCarOffset)
    Offset Amount=1
    Ribbon Width=28
   )
   &AddVelocity(Velocity=&VC_Mask_Vectors(Coro Value:Vector3f=<-800 0 0>, FrontEnd Value:Vector3f=<-4000 0 0>, FN Value:Vector3f=<-1000 0 0>)) {module=AddVelocity001}
   &VC_ExtraBindingColors(
    ExtraColors=3
    R1 Color:LinearColor=&MakeCustomLinearColorFromBool(Bool=$User.bIsPainted, True Color=$User.PaintedColor, False Color=<rgba 0.1 0.08 1 1>)
    R2 Color:LinearColor=&MakeCustomLinearColorFromBool(Bool=$User.bIsPainted, True Color=$User.PaintedColor, False Color=<rgba 2 0.2 0.1 1>)
    R3 Color:LinearColor=&MakeCustomLinearColorFromBool(Bool=$User.bIsPainted, True Color=$User.PaintedColor, False Color=<rgba 0.2 0.5 1 1>)
   )
  ]
  particleUpdateScript [
   /SetVariables(
    Particles.BlueLightningVelocity:Vector3f=&Multiply_VectorByFloat(Vector=$Particles.Velocity, Float=0)
    Particles.LightningSmallRW:NiagaraFloat=&Multiply_FloatByInt(Float=$Emitter.LightningSmallRW, Integer=1)
    Particles.LightningsRW:NiagaraFloat=&Multiply_FloatByInt(Float=$Emitter.LightningsRW, Integer=1)
    Particles.R1BlueLightningColor:LinearColor=&self:VC_Dynamic_ScaleColor(
     Color to Scale:LinearColor=$Particles.Initial.R1 Color
     RGB Scale:Vector3f=&MakeCustomVectorFromBool(
      Bool=$User.bIsPainted
      True Vector=&Multiply_VectorByFloat(Vector=<8 8 8>, Float=$User.LightningBrightness)
      False Vector=<1 1 1>
     )
     Alpha Scale:NiagaraFloat=&Multiply_Float(
      A=&FloatFromCurve(FloatCurve=DataInterface() {propertyValues {Curve {keys [<at 0 1.0113 RCIM_Cubic>, {arriveTangent=-1.0113, interpMode=RCIM_Cubic, leaveTangent=-1.0113, time=0.5, value=0.5}, <at 1 0 RCIM_Cubic>]}}})
      B=&MakeCustomFloatFromBool(Bool=$User.bIsPainted, True Float=0.1, False Float=3)
     )
    )
    Particles.R2RedLightningColor:LinearColor=&self:VC_Dynamic_ScaleColor(
     Color to Scale:LinearColor=$Particles.Initial.R2 Color
     RGB Scale:Vector3f=&MakeCustomVectorFromBool(
      Bool=$User.bIsPainted
      True Vector=&Multiply_VectorByFloat(Vector=<2.5 2.5 2.5>, Float=$User.LightningBrightnessRed)
      False Vector=<1 1 1>
     )
     Alpha Scale:NiagaraFloat=&Multiply_Float(
      A=&FloatFromCurve(FloatCurve=DataInterface() {propertyValues {Curve {keys [<at 0 1.0113 RCIM_Cubic>, {arriveTangent=-1.0113, interpMode=RCIM_Cubic, leaveTangent=-1.0113, time=0.5, value=0.5}, <at 1 0 RCIM_Cubic>]}}})
      B=0.1
     )
    )
    Particles.R3BlueLightningColor:LinearColor=&self:VC_Dynamic_ScaleColor(
     Color to Scale:LinearColor=$Particles.Initial.R3 Color
     RGB Scale:Vector3f=&MakeCustomVectorFromBool(Bool=$User.bIsPainted, True Vector=<10 10 10>, False Vector=<10 10 10>)
     Alpha Scale:NiagaraFloat=&Multiply_Float(
      A=&FloatFromCurve(FloatCurve=DataInterface() {propertyValues {Curve {keys [<at 0 1.0113 RCIM_Cubic>, {arriveTangent=-1.0113, interpMode=RCIM_Cubic, leaveTangent=-1.0113, time=0.5, value=0.5}, <at 1 0 RCIM_Cubic>]}}})
      B=0.1
     )
    )
    Particles.R3BlueLightningPosition:NiagaraPosition=&Multiply_Vector(
     A=Linked() {linkedVariable {name=Particles.Position, type=NiagaraPosition}}
     B=&ConvertPositionToVector(Input Position=<3 1 1>, Conversion Operation:ENiagara_LWCConvertPosToVec=::NewEnumerator0)
    )
    Particles.R3LightningRedRW:NiagaraFloat=&Multiply_Float(A=$Emitter.LightningRedRW, B=&RandomRangeFloat(Minimum=0.9, Maximum=1.2))
   ) {isSetParameters=true}
   &ParticleState()
   &ScaleColor(
    Scale Alpha=&Multiply_Float(
     A=&FloatFromCurve(FloatCurve=DataInterface() {propertyValues {Curve {keys [<at 0 1 RCIM_Cubic>,<at 1 0 RCIM_Cubic>]}}})
     B=0.22
    )
    Scale RGB=&MakeCustomVectorFromBool(Bool=$User.bIsPainted, True Vector=<2.5 2.5 2.5>, False Vector=<2.5 2.5 2.5>)
    ScaleA=true
    ScaleRGB=true
   )
   &ScaleColor(Scale RGB=&Multiply_VectorByFloat(Vector=<1 1 1>, Float=$User.Intensity), ScaleA=true, ScaleRGB=true) {enabled=false, module=ScaleColor001}
   &InheritSourceMovement(
    Applied Position Scale=&Multiply_VectorByFloat(
     Vector=<1 1 1>
     Float=&FloatFromCurve(FloatCurve=DataInterface() {propertyValues {Curve {keys [{interpMode=RCIM_Cubic, tangentMode=RCTM_Break, time=0, value=1}, {interpMode=RCIM_Cubic, tangentMode=RCTM_Break, time=1, value=0}]}}})
    )
    ApplyForce=false
    ApplyPosition=true
    ApplyPositionFalloff=false
   )
   &DynamicMaterialParameters(
    Index 0 Param 1=3
    Index 0 Param 2=&RandomRangeFloat(Minimum=0, Maximum=1)
    Index 1 Param 1=&RandomRangeFloat(Minimum=2, Maximum=5)
    Index 1 Param 2=&RandomRangeFloat(Minimum=0, Maximum=1)
    Index 2 Param 1=&RandomRangeFloat(Minimum=1.5, Maximum=2)
    Index 2 Param 2=&RandomRangeFloat(Minimum=0, Maximum=1)
    Index 3 Param 1=3
    Index 3 Param 2=1
    Index 3 Param 3=3
    Module.Param0WriteEnabled=true
    Module.Param0WriteEnabled1=true
    Module.Param0WriteEnabled2=true
    Module.Param0WriteEnabled3=true
    Module.Param1WriteEnabled=true
    Module.Param1WriteEnabled1=true
    Module.Param1WriteEnabled2=true
    Module.Param1WriteEnabled3=true
    Module.Param2WriteEnabled=true
    Module.Param2WriteEnabled1=true
    Module.Param2WriteEnabled2=true
    Module.Param2WriteEnabled3=true
    Module.Param3WriteEnabled=true
    Module.Param3WriteEnabled1=true
    Module.Param3WriteEnabled2=true
    Module.Param3WriteEnabled3=true
    Write Parameter Index 1=true
    Write Parameter Index 2=true
    Write Parameter Index 3=true
   )
   &ScaleRibbonWidth(Ribbon Width Scale=&FloatFromCurve(FloatCurve=DataInterface() {propertyValues {Curve {keys [<at 0 3>,<at 0.65 1>]}}}))
   &SolveForcesAndVelocity(Clamp Velocity=false, Limit Acceleration=false)
   &VC_UpdateRibbon_v1(Enable Skeletal Position=true)
  ]
  rendererClasses [&NiagaraRibbonRendererProperties]
  renderers [
   {rendererClass=&NiagaraRibbonRendererProperties}
   {rendererClass=&NiagaraRibbonRendererProperties, rendererIndex=1}
   {rendererClass=&NiagaraRibbonRendererProperties, rendererIndex=2}
  ]
 }
 Sparks001 {
  emitterUpdateScript [
   &EmitterState()
   &VC_FilteredSocketCount()
   &SpawnRate(SpawnRate=35, Use Spawn Probability=false)
   &SpawnRate(SpawnGroup=1, SpawnRate=35, Use Spawn Probability=false) {module=SpawnRate001}
  ]
  particleSpawnScript [
   &VC_PaintedVariantsOverride() {enabled=false}
   &InitializeParticle(
    Color=&MakeCustomLinearColorFromBool(
     Bool=$User.bIsPainted
     True Color=$User.PaintedColor
     False Color=&MaskLinearColorBySpawnGroup(Masked Linear Color=<rgba 1 0.1 0.05 1>, Passthrough Linear Color=<rgba 0.2 0.5 1 1>)
    )
    Lifetime Max=0.4
    Lifetime Min=0.2
    Lifetime Mode=::NewEnumerator1
    Position Offset=<2 0 0>
    Sprite Size Mode=::NewEnumerator1
    Uniform Sprite Size Max=20
    Uniform Sprite Size Min=4
    UsePositionOffset=true
    Write Material Random=true
   )
   &PartitionParticles(Number Of Partitions=$Emitter.PortCount)
   &SkeletalMeshLocation(
    Bone / Socket Index=$Particles.PartitionParticles.Partition
    Bone / Socket Sampling Mode=::NewEnumerator5
    MaskSpawnGroup=false
    Mesh Sampling Type=::NewEnumerator6
    Sampled Position Offset=&AddVectorToPosition(Position=<-20 0 0>, Vector=$User.DTCarOffset)
    Skeletal Mesh=$System.SkeletalMesh
   )
   &ShapeLocation(
    MaskSpawnGroup=false
    Offset Mode=::NewEnumerator0
    Override Seed=false
    Rotation Coordinate Space=::Simulation
    Rotation Mode=::NewEnumerator1
    Sphere Radius=20
    UseOwnerScale=true
   )
   &AddVelocity(
    Velocity=&Multiply_VectorByFloat(
     Vector=&MaskVectorBySpawnGroup(
      Masked Vector=&RandomRangeVector(Minimum=<150 100 100>, Maximum=<1400 -100 -100>)
      Passthrough Vector=&RandomRangeVector(Minimum=<700 140 140>, Maximum=<1400 -140 -140>)
     )
     Float=&VC_Mask_Floats(Coro Value:NiagaraFloat=0.25, FrontEnd Value:NiagaraFloat=0.5, FN Value:NiagaraFloat=0.5)
    )
    Velocity Speed Scale=&MaskFloatBySpawnGroup(Masked Float=-2, Passthrough Float=-4)
   )
   &VC_ExtraBindingColors(
    ExtraColors=1
    R1 Color:LinearColor=&MakeCustomLinearColorFromBool(Bool=$User.bIsPainted, True Color=$User.PaintedColor, False Color=<rgba 0.2 0.5 1 1>)
   )
   /SetVariables(Particles.BlueSparkVisTag:NiagaraInt32=&SpawnGroup()) {isSetParameters=true}
   /VehicleCosmetics/FX/VC_FXParents/Modules/DLR_ProductPreviews(GameModeID:NiagaraInt32=$System.GameModeID, bPreviewVehicleSpeed:NiagaraBool=$System.bPreviewVehicleSpeed)
  ]
  particleUpdateScript [
   &ParticleState()
   &ScaleColor(
    Scale Alpha=&Multiply_Float(
     A=&FloatFromCurve(FloatCurve=DataInterface() {propertyValues {Curve {keys [<at 0 1.0113 RCIM_Cubic>, {arriveTangent=-1.0113, interpMode=RCIM_Cubic, leaveTangent=-1.0113, time=0.5, value=0.5}, <at 1 0 RCIM_Cubic>]}}})
     B=0.1
    )
    Scale RGB=&MaskVectorBySpawnGroup(Masked Vector=<10 10 10>, Passthrough Vector=<100 100 100>)
    ScaleA=true
    ScaleRGB=true
   ) {module=ScaleColor001}
   &SubUVAnimation(
    End Frame Range Override=16
    "Sprite Renderer "=DataInterface() {propertyValues {SpriteRenderer=&self:Sparks001_1.NiagaraSpriteRendererProperties_0}}
    UseEndFrame=true
    UseStartFrame=true
   )
   &DynamicMaterialParameters(
    Index 0 Param 1=10
    Index 1 Param 1=1
    Module.Param0WriteEnabled=true
    Module.Param0WriteEnabled1=true
    Module.Param1WriteEnabled=true
    Module.Param1WriteEnabled1=true
    Module.Param2WriteEnabled=true
    Module.Param2WriteEnabled1=true
    Module.Param3WriteEnabled=true
    Module.Param3WriteEnabled1=true
    Write Parameter Index 1=true
   )
   &Drag(
    Drag=&MaskFloatBySpawnGroup(Masked Float=0, Passthrough Float=15)
    Use Linear Drag=true
    Use Rotational Drag=true
   )
   &InheritSourceMovement(
    Applied Position Scale=&Multiply_VectorByFloat(
     Vector=<1 1 1>
     Float=&FloatFromCurve(FloatCurve=DataInterface() {propertyValues {Curve {
         defaultValue=1
         keys [
          {arriveTangent=0.0816, interpMode=RCIM_Cubic, leaveTangent=0.0816, tangentMode=RCTM_User, time=0.5, value=1}
          {arriveTangent=-3.5615, interpMode=RCIM_Cubic, leaveTangent=-3.5615, tangentMode=RCTM_User, time=1, value=0}
         ]
        }}})
    )
    ApplyForce=false
    ApplyPosition=true
    ApplyPositionFalloff=false
   ) {module=InheritSourceMovement001}
   &SolveForcesAndVelocity(Clamp Velocity=false, Limit Acceleration=false)
  ]
  rendererClasses [&NiagaraSpriteRendererProperties]
  renderers [
   {rendererClass=&NiagaraSpriteRendererProperties}
   {rendererClass=&NiagaraSpriteRendererProperties, rendererIndex=1}
  ]
 }
}
summary {
 systemName=NS_Boost_Zti
 emitters [
  {emitterName=BlueLightning001, rendererClasses [&NiagaraRibbonRendererProperties]}
  {emitterName=Ribbons, rendererClasses [&NiagaraRibbonRendererProperties]}
  {emitterName=Lightning_Sprites, rendererClasses [&NiagaraSpriteRendererProperties]}
  {emitterName=LightningBody_Sprites, rendererClasses [&NiagaraSpriteRendererProperties]}
  {emitterName=Sparks001, rendererClasses [&NiagaraSpriteRendererProperties]}
 ]
 userVariables [
  User.bIsFrontendPreview {type=NiagaraBool, defaultValue=true}
  User.bIsPainted {type=NiagaraBool}
  User.blsFrontEndPreview {type=NiagaraBool}
  User.bPreviewVehicleSpeed {type=NiagaraBool}
  User.ColorVariantID {type=NiagaraInt32}
  User.DTCarOffset {type=Vector3f, description="", defaultValue {struct=Vector3f, value=<0 0 0>}}
  User.GameModeID {type=NiagaraInt32}
  User.Intensity {type=NiagaraFloat, defaultValue=1, description=""}
  User.LightningBrightness {type=NiagaraFloat, defaultValue=1, description=""}
  User.LightningBrightnessRed {type=NiagaraFloat, defaultValue=1, description=""}
  User.PaintedColor {type=LinearColor, description="", defaultValue {struct=LinearColor, value=<rgba 1 1 1 1>}}
  User.PaintedScalar {type=NiagaraFloat, defaultValue=1, description=""}
  User.PortCount {type=NiagaraInt32, defaultValue=2}
 ]
}
