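# Translucent unlit master material for lightning effects on a vehicle boost cosmetic, built for Niagara ribbons and sprites.
# A time-driven, optionally rotated flipbook samples a packed RGB texture, with static switches choosing a single channel or a lerped, multiplied or added blend.
# An optional panning gradient mask modulates or adds to the lightning before a U-axis edge fade.
# Particle color and dynamic Brightness and LerpBlend parameters drive emissive and opacity.
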Material /VehicleCosmetics/Boosts/Boost_Zti/Materials/M_Boost_Lightning_MAster
graph {
 properties {
  bUsedWithNiagaraRibbons=true
  bUsedWithNiagaraSprites=true
  blendMode=BLEND_Translucent
  pixelDepthOffsetMode=PDOM_Legacy
  shadingModel=MSM_Unlit
  usedShadingModels=Unlit
 }
 nodes [
  n1 = TextureCoordinate() {uTiling=0.05, vTiling=0.25}
  n2 = Constant() {r=1}
  n3 = Time()
  n4 = ScalarParameter() {defaultValue=1, parameterName=PanSpeed}
  n5 = StaticSwitchParameter(
   True=Rotator(Coordinate=Panner(Coordinate=@n1, Time=@n2) {speedY=-0.4}, Time=Constant() {r=0.5}) {speed=3.14}
   False=Panner(Coordinate=TextureCoordinate() {uTiling=0.25, vTiling=0.05}, Time=Multiply(A=@n3, B=@n4)) {speedX=-0.4}
  ) {parameterName=RotateFlipbook}
  n6 = Panner(Coordinate=TextureCoordinate() {vTiling=0.5}, Time=Constant() {r=1}) {speedY=0.25}
  n7 = Constant() {r=0.5}
  n8 = ScalarParameter() {defaultValue=4, parameterName=Columns}
  n9 = ScalarParameter() {defaultValue=4, parameterName=Rows}
  n10 = AppendVector(A=@n8, B=@n9)
  n11 = Divide(A=StaticSwitchParameter(True=Rotator(Coordinate=@n6, Time=@n7) {speed=3.14}, False=@n6) {parameterName=RotateFlipbook}, B=@n10)
  n12 = AppendVector(A=Multiply(A=@n8, B=@n9), B=@n9)
  n13 = Frac(None=Multiply(A=Time(), B=ScalarParameter() {defaultValue=1, parameterName=Speed}))
  n14 = AppendVector(A=@n13, B=@n13)
  n15 = Divide(A=Floor(None=Multiply(A=@n12, B=@n14)), B=@n10)
  n16 = TextureCoordinate()
  n17 = Rotator(Coordinate=@n16, Time=Constant() {r=0.5}) {speed=3.14}
  n18 = TextureSample(UVs=StaticSwitchParameter(True=Add(A=@n11, B=@n15), False=StaticSwitchParameter(True=@n17, False=@n16) {parameterName=RotateFlipbook}) {defaultValue=true, parameterName=FlipBook}) {texture=/DealershipSharedLibrary/FX/Textures/TexturePackage/T_VCVFX_ZTITrailFB_RGB}
  n19 = StaticComponentMaskParameter(None=@n18.RGB) {defaultG=true, parameterName=Blend_01}
  n20 = StaticComponentMaskParameter(None=@n18.RGB) {defaultB=true, parameterName=Blend_02}
  n21 = DynamicParameter() {paramNames [Brightness,LerpBlend,Param3,Param4]}
  n22 = Multiply(A=@n19, B=@n20)
  n23 = Add(A=@n19, B=@n20)
  n24 = StaticSwitchParameter(
   True=StaticSwitchParameter(
    True=LinearInterpolate(A=@n19, B=@n20, Alpha=@n21.LerpBlend)
    False=StaticSwitchParameter(True=@n22, False=@n23) {parameterName="BlendType(T=M)(F=A)"}
   ) {parameterName=LerpBlend}
   False=StaticComponentMaskParameter(None=@n18.RGB) {defaultG=true, parameterName=RGB_Mask}
  ) {parameterName=Blend}
  n25 = Multiply(
   A=Multiply(
    A=ScalarParameter() {defaultValue=3, parameterName=PanBrightness}
    B=TextureSample(UVs=@n5) {texture=/DealershipSharedLibrary/FX/Textures/Gradient_Masks/T_VCVFX_Time_Mask_A}.RGB
   )
   B=@n24
  )
  n26 = StaticSwitchParameter(True=StaticSwitchParameter(True=Add(A=@n25, B=@n24), False=@n25) {parameterName=PanAdd}, False=@n24) {parameterName=PanLightning}
  n27 = TextureCoordinate() {uTiling=4, vTiling=4}
  n28 = ComponentMask(None=StaticSwitchParameter(True=TextureCoordinate(), False=Frac(None=@n27)) {defaultValue=true, parameterName=FlipBook}) {r=true}
  n29 = OneMinus(None=@n28)
  n30 = Multiply(A=Multiply(A=Multiply(A=@n26, B=@n28), B=@n21.Brightness), B=@n29)
  n31 = Multiply(A=@n29, B=Constant() {r=8})
  n32 = ParticleColor()
  MaterialFunctionCall(Emissive=Multiply(A=Multiply(A=@n30, B=@n31), B=@n32.RGB)) {materialFunction=&MF_FX_EmissiveAdjustments}
  Saturate(None=Multiply(A=@n30, B=@n32.A))
 ]
}
